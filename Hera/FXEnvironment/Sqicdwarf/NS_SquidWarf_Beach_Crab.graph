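# Niagara GPU system that scatters night-time beach crabs around the emitter, with distance culling and camera-distance scaling.
# Crab meshes follow sampled landscape height and normal, wandering on noise forces while an attraction force keeps them within the radius.
# Player proximity or a randomized timeline outside night triggers burrowing: crabs wiggle, sink and die beneath a time-of-day-tinted sand sprite.
# Velocity drives material parameters, and user parameters set radius, size and spawn rate.
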NiagaraSystem /Hera/FXEnvironment/Sqicdwarf/NS_SquidWarf_Beach_Crab
dependencies {
 usedDataInterfaces [
  &NiagaraDataInterfaceCurve
  &NiagaraDataInterfaceCamera
  &NiagaraDataInterfaceSpriteRendererInfo
 ]
 usedDynamicInputs []
 usedModules [&SystemState]
}
emitters {Crabs {
  simTarget=GPUComputeSim
  emitterSpawnScript [/SetVariables(Emitter.GlobalRadius:NiagaraFloat=&Multiply_Float(A=300, B=$User.RadiusMult)) {isSetParameters=true}]
  emitterUpdateScript [
   &EmitterState(
    Enable Distance Culling=true
    Life Cycle Mode=::NewEnumerator1
    Loop Behavior=::NewEnumerator1
    Loop Duration Mode=::NewEnumerator1
    Max Distance Response=::SleepAndClearParticles
    Min Distance Response=::SleepAndLetParticlesFinish
    Module.bApplySpawnCountScaleByDistance=true
    Module.bUseMaxDistance=true
    Module.bUseMinDistance=true
    Reset Age On Awaken=true
    Scalability Mode=::NewEnumerator1
    Scale Spawn Count=true
    Spawn Count Scale By Distance=DataInterface() {propertyValues {Curve {keys [<at 0 1>,<at 0.95 1>,<at 1 0>]}}}
    UseLoopDelay=false
   )
   &SpawnRate(
    SpawnRate=&FloatFromCurve(
     FloatCurve=DataInterface() {propertyValues {Curve {keys [<at 0 0>,<at 0.1 1>,<at 0.2 0>]}}}
     CurveIndex=$Emitter.Age
     Scale Curve=&Multiply_Float(
      A=&FNTimeOfDayScalar(Night Scale=1, Morning Scale=0, Day Scale=0, Evening Scale=0)
      B=&Multiply_Float(A=200, B=$User.SpawnRate)
     )
    )
    Use Spawn Probability=false
   )
  ]
  particleSpawnScript [
   &InitializeParticle(
    AdjustAlpha=false
    AdjustHue=true
    AdjustSaturation=true
    AdjustValue=true
    Color Mode=::NewEnumerator3
    Hue Shift Range:Vector2f=<30 60>
    Mesh Scale Mode=::NewEnumerator3
    Mesh Uniform Scale=&Multiply_Float(
     A=&Lerp_Float(
      A=1.35
      B=2
      Alpha=&FloatFromCurve(
       FloatCurve=DataInterface() {propertyValues {Curve {
          defaultValue=1
          keys [
           {interpMode=RCIM_Cubic, tangentMode=RCTM_User, time=0, value=0}
           {arriveTangent=0.2471, interpMode=RCIM_Cubic, leaveTangent=0.2471, tangentMode=RCTM_User, time=0.4663, value=0.0496}
           {arriveTangent=2.4745, interpMode=RCIM_Cubic, leaveTangent=2.4745, tangentMode=RCTM_User, time=0.9047, value=0.5052}
           {arriveTangent=5.9789, interpMode=RCIM_Cubic, leaveTangent=5.9789, tangentMode=RCTM_User, time=1, value=1}
          ]
         }}}
       CurveIndex=&RandomRangeFloat(Minimum=0, Maximum=1)
      )
     )
     B=&Add_Float(
      A=Linked() {linkedVariable {name="User.Size  Scale", type=NiagaraFloat}}
      B=&RandomRangeFloat(Minimum=$User.Size Variance, Maximum=&Multiply_Float(A=1, B=-1))
     )
    )
    UsePositionOffset=false
    Value Range:Vector2f=<3 6>
    Write Material Random=true
   )
   &ShapeLocation(
    Cylinder Radius=$Emitter.GlobalRadius
    MaskSpawnGroup=false
    Offset Mode=::NewEnumerator0
    Override Seed=false
    Shape Primitive=::NewEnumerator1
    Surface Only=false
    Surface Only Band Thickness=1
    UseOwnerScale=true
   )
   &InitialMeshOrientation(
    Mesh Axis To Orient=<0 0 1>
    Mesh Orientation Mode=::NewEnumerator2
    Rotation=&RandomRangeVector(Minimum=<0 0 -0.05>, Maximum=<0 0 0.05>)
    Use Rotation Vector=true
   )
   /SetVariables() {isSetParameters=true}
   /SetVariables(Particles.SandSpriteSize:Vector2f=&Vector2DFromFloat(Value=&Multiply_Float(
      A=80
      B=&Max_Float(
       A:NiagaraFloat=&MakeFloatFromVector(Vector=$Particles.Scale, Channel=::NewEnumerator0)
       B:NiagaraFloat=&Max_Float(
        A:NiagaraFloat=&MakeFloatFromVector(Vector=$Particles.Scale, Channel=::NewEnumerator1)
        B:NiagaraFloat=&MakeFloatFromVector(Vector=$Particles.Scale, Channel=::NewEnumerator2)
       )
      )
     ))) {isSetParameters=true}
   &AddVelocity(Velocity=&RandomRangeVector(Minimum=<-35 -35 0>, Maximum=<35 35 1>))
  ]
  particleUpdateScript [
   &ParticleState(Kill Particles When Lifetime Has Elapsed=false, Let Infinitely Lived Particles Die When Emitter Deactivates=true)
   /SetVariables(Particles.DistanceToCenter:NiagaraFloat=&DistanceBetweenPositions(A=$Particles.Position, B=$Engine.Emitter.SimulationPosition)) {isSetParameters=true}
   /Niagara/Modules/Update/Utility/VectorNoiseValue(
    Noise Frequency=25
    Pan Noise=true
    Pan Noise Field:Vector3f=<0 0 0.1>
    Remap Output Value:ENiagara_CurlNoiseRemapType=::NewEnumerator2
    Remap Range Output Max:Vector3f=<0 0 0>
    Remap Range Output Min:Vector3f=<1 0 0>
   )
   &LinearForce(Force=&Multiply_VectorByFloat(
     Vector=$Output.VectorNoiseValue.Noise
     Float=&Multiply_Float(
      A=&RandomRangeFloat(Minimum=0, Maximum=1)
      B=&MakeCustomFloatFromBool(Bool=$Particles.IsBurrowing, True Float=0, False Float=100)
     )
    ))
   &PointAttractionForce(
    Attraction Radius:NiagaraFloat=&Multiply_Float(A=$Emitter.GlobalRadius, B=2)
    Attraction Strength=&MakeCustomFloatFromBool(
     Bool=&SetBoolByFloatComparison(A=$Particles.DistanceToCenter, B=$Emitter.GlobalRadius)
     True Float=&MakeCustomFloatFromBool(Bool=$Particles.IsBurrowing, True Float=0, False Float=3)
     False Float=0
    )
    Kill Within Radius=false
    Module.Use Falloff=true
    UseAttractorOffset=false
   )
   /Niagara/Modules/Landscape/SampleLandscape(Facing Axis:Vector3f=<0 0 1>, Height Offset=5)
   &UpdateMeshOrientation(
    Facing Direction=$Output.SampleLandscape.Normal
    Facing Interpolation=true
    Facing Interpolation Rate=0
    Orientation Method=::NewEnumerator1
    Override Quat=false
    Stabilization Mode:ENiagara_OrientVectorCount=::NewEnumerator0
    TransformFacing=false
   )
   &Drag(Use Linear Drag=true, Use Rotational Drag=true)
   &CurlNoiseForce(
    Bias Noise=false
    Noise Frequency=50000
    Noise Strength=&ClampFloat(
     Float=&Waveform(
      "Waveform [1]"=::NewEnumerator12
      "[1] Frequency"=&RandomRangeFloat(Minimum=0.1, Maximum=3)
      "[1] Amplitude Min/Max"=<-350 300>
     )
     Max=300
    )
    Pan Noise=true
    Pan Noise Field=&RandomVector(Vector Scale=0.1)
   )
   &SolveForcesAndVelocity(Clamp Velocity=false, Limit Acceleration=false)
   &Timeline(
    Play=&SetBoolByFloatComparison(A=&FNTimeOfDayScalar(Night Scale=0, Morning Scale=1, Evening Scale=1), B=0.5)
    Play Rate=&Divide_Float(A=1, B=&RandomRangeFloat(Minimum=0.35, Maximum=2))
   ) {module=Timeline001}
   /SetVariables(Particles.IsBurrowing:NiagaraBool=&Boolean_OrOperation(
     Bool1=&SetBoolByFloatComparison(A=$StackContext.Timeline001.NormalizedTime, B=0.99)
     Bool2=&Boolean_OrOperation(
      Bool1=$Particles.IsBurrowing
      Bool2=&SetBoolByFloatComparison(
       A=800
       B=&DistanceBetweenPositions(
        A=&TransformPosition(Position=$Particles.Position, Source Space=::Local, Destination Space=::World)
        B=&MakeVectorFromLinearColorRGB(Color=$NPC.FortniteNPC.PlayerPosition)
       )
       Comparison Type=::NewEnumerator2
      )
     )
    )) {isSetParameters=true}
   &Timeline(Play=$Particles.IsBurrowing)
   &UpdateMeshOrientation(
    Rotation Rate=&Multiply_Float(A=&Sine(Normalized Angle=$Emitter.Age, Period=0.3, Scale=2, Bias=0), B=$Particles.Timeline.NormalizedTime)
    Rotation Vector=<0 0 1>
   ) {module=UpdateMeshOrientation001}
   &KillParticles(Kill Particles=&SetBoolByFloatComparison(A=1, B=$Particles.Timeline.Time, Comparison Type=::NewEnumerator4))
   /SetVariables(Particles.Position=&AddVectorToPosition(
     Position=$Particles.Position
     Vector=&Multiply_VectorByFloat(
      Vector=&RandomRangeVector(Minimum=<0 0 -20>, Maximum=<0 1 -25>)
      Float=$Particles.Timeline.NormalizedTime
     )
    )) {isSetParameters=true}
   /SetVariables(Particles.Speed:NiagaraFloat=&RemapRange(
     Input Value To Remap=&VectorLength(Vector=$Particles.Velocity)
     Input Min=0
     Input Max=200
     Output Min=0
     Output Max=1
    )) {isSetParameters=true}
   &ScaleMeshSize(Scale Factor=&Multiply_VectorByFloat(
     Vector=$Particles.Initial.Scale
     Float=&OneMinusFloat(Float=/Niagara/DynamicInputs/Camera/CameraFade(
       Camera Query:NiagaraDataInterfaceCamera=DataInterface()
       Fade Offset:NiagaraFloat=1500
       Fade Distance:NiagaraFloat=5000
       Use Owner Position:NiagaraBool=true
      ))
    ))
   &DynamicMaterialParameters(
    Index 0 Param 1=&RandomRangeFloat(Minimum=1, Maximum=1)
    Index 0 Param 2=&ClampFloat(
     Float=&Multiply_Float(
      A=$Particles.Speed
      B=&Multiply_Float(A=&VectorLength(Vector=$Particles.Scale), B=50)
     )
     Max=1.5
    )
    Index 0 Param 3=4
    Index 0 Param 4=&Multiply_Float(A=$Particles.Speed, B=0.25)
    Index 1 Param 1=&Multiply_Float(A=$Particles.Speed, B=0.25)
    Module.Index0Float0Write=true
    Module.Index0Float1Write=true
    Module.Index0Float2Write=true
    Module.Index0Float3Write=true
    Module.Index1Float0Write=true
    Module.Index1Float1Write=true
    Module.Index1Float2Write=true
    Module.Index1Float3Write=true
    Module.Index2Float0Write=false
    Module.Index2Float1Write=false
    Module.Index2Float2Write=false
    Module.Index2Float3Write=false
    Module.Index3Float0Write=false
    Module.Index3Float1Write=false
    Module.Index3Float2Write=false
    Module.Index3Float3Write=false
    Write Parameter Index 1=::NewEnumerator1
    Write Parameter Index 2=::NewEnumerator1
    Write Parameter Index 3=::NewEnumerator1
   )
   /SetVariables(Particles.SandPosition:NiagaraPosition=&MakeVector(
     X=&MakeFloatFromVector(Vector=Linked() {linkedVariable {name=Particles.Position, type=NiagaraPosition}}, Channel=::NewEnumerator0)
     Y=&MakeFloatFromVector(Vector=Linked() {linkedVariable {name=Particles.Position, type=NiagaraPosition}}, Channel=::NewEnumerator1)
     Z=&Add_Float(
      A=&MakeFloatFromVector(Vector=Linked() {linkedVariable {name=Output.SampleLandscape.Position, type=NiagaraPosition}}, Channel=::NewEnumerator2)
      B=-10
     )
    )) {isSetParameters=true}
   /SetVariables(Particles.SandCamera=50) {isSetParameters=true}
   /SetVariables(Particles.SandColor:LinearColor=&MultiplyLinearColor(
     Linear Color=&FNTimeOfDayColor(
      Night Color=<rgba 0.2154 0.2533 0.3217 1>
      Morning Color=<rgba 0.5271 0.2918 0.2961 1>
      Day Color=<rgba 0.4261 0.3042 0.2541 1>
      Evening Color=<rgba 0.3304 0.1892 0.1868 1>
     )
     Scale Factor=&MakeFloat4FromVectorAndFloat(
      "Vector (RGB)"=<1 1 1>
      "Float (Alpha)"=&FloatFromCurve(
       FloatCurve=DataInterface() {propertyValues {Curve {
          defaultValue=0
          keys [
           {interpMode=RCIM_Cubic, tangentMode=RCTM_User, time=0, value=0}
           {interpMode=RCIM_Cubic, tangentMode=RCTM_User, time=0.5, value=1}
           {tangentMode=RCTM_User, time=1, value=0}
          ]
         }}}
       CurveIndex=$Particles.Timeline.NormalizedTime
      )
     )
    )) {isSetParameters=true}
   &SubUVAnimation(
    "Sprite Renderer "=DataInterface() {propertyValues {SpriteRenderer=&self:Crabs_0.NiagaraSpriteRendererProperties_2}}
    SubUV Animation Mode:ENiagara_SubUVLookupModeV2=::NewEnumerator1
    UseEndFrame=false
    UseStartFrame=false
   )
  ]
  rendererClasses [&NiagaraMeshRendererProperties, &NiagaraSpriteRendererProperties]
  renderers [
   {rendererClass=&NiagaraMeshRendererProperties, rendererIndex=0}
   {rendererClass=&NiagaraMeshRendererProperties, rendererIndex=1}
   {rendererClass=&NiagaraSpriteRendererProperties, rendererIndex=2}
  ]
 }}
summary {
 systemName=NS_SquidWarf_Beach_Crab
 emitters [{
   emitterName=Crabs
   simTarget=GPUComputeSim
   rendererClasses [&NiagaraMeshRendererProperties, &NiagaraSpriteRendererProperties]
  }]
 userVariables [
  User.Landscape {
   type=NiagaraDataInterfaceLandscape
   description=""
   defaultValue {
    struct=NiagaraExt_VariableValue_DataInterface
    value {
     dataInterface=&self:NiagaraDataInterfaceLandscape_0
     dataInterfaceClass=&NiagaraDataInterfaceLandscape
    }
   }
  }
  User.RadiusMult {type=NiagaraFloat, defaultValue=3, description=""}
  "User.Size  Scale" {type=NiagaraFloat, defaultValue=1, description=""}
  User.Size Variance {type=NiagaraFloat, defaultValue=0, description=""}
  User.SpawnRate {type=NiagaraFloat, defaultValue=0.4, description=""}
 ]
}
